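# Sound cue for the DoubleDuty glider's thrust loop, layering a continuous thrust sound with intermittent idle growls.
# The AnyDirection parameter scales the looping thrust wave's volume and pitch from reduced minimums.
# A quieter layer loops one of five random growls through a random delay and modulator, with inverted AnyDirection volume so growls fade out as thrust rises.
# Glider thrust attenuation and a cosmetic glider loop properties sheet govern playback.
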SoundCue /BRCosmetics/Sounds/Gliders/DoubleDuty/GliderThrustLoop_DoubleDuty_Cue
root=Mixer_0
nodes [
 {id=Delay_0, children [Random_0], properties {delayMax=10, delayMin=5}}
 {id=Looping_0, children [ModulatorContinuous_1]}
 {id=Mixer_0, children [ModulatorContinuous_0, Looping_0], properties {inputVolume [1,0.3]}}
 {
  id=ModulatorContinuous_0
  children [WavePlayer_0]
  properties {
   pitchModulationParams {default=0, minOutput=0.9, parameterName=AnyDirection}
   volumeModulationParams {default=0, minOutput=0.2, parameterName=AnyDirection}
  }
 }
 {
  id=ModulatorContinuous_1
  children [Modulator_0]
  properties {
   pitchModulationParams {default=0, minOutput=1, parameterName=AnyDirection}
   volumeModulationParams {default=0, maxOutput=0, minOutput=1, parameterName=AnyDirection}
  }
 }
 {id=Modulator_0, children [Delay_0]}
 {id=Random_0, children [WavePlayer_1,WavePlayer_2,WavePlayer_3,WavePlayer_4,WavePlayer_5], properties {weights [1,1,1,1,1]}}
 {id=WavePlayer_0, properties {bLooping=true, soundWaveAssetPtr=&/IP_Glider_DoubleDuty_Thurst_LP}}
 {id=WavePlayer_1, properties {soundWaveAssetPtr=&/IP_Glider_DoubleDuty_Idle_Growl_01}}
 {id=WavePlayer_2, properties {soundWaveAssetPtr=&/IP_Glider_DoubleDuty_Idle_Growl_02}}
 {id=WavePlayer_3, properties {soundWaveAssetPtr=&/IP_Glider_DoubleDuty_Idle_Growl_03}}
 {id=WavePlayer_4, properties {soundWaveAssetPtr=&/IP_Glider_DoubleDuty_Idle_Growl_04}}
 {id=WavePlayer_5, properties {soundWaveAssetPtr=&/IP_Glider_DoubleDuty_Idle_Growl_05}}
]
properties {
 attenuationSettings=&Fort_Player_Glider_Thrust_Athena_Att
 audioPropertiesSheet=&APS_Cosmetic_Glider_Loop
 maxDistance=12500
 soundClassObject=&Fort_Gliders_Land
 volumeMultiplier=0.9
 concurrencySet [null]
}
